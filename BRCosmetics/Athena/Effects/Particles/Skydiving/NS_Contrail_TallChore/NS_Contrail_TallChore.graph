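# Niagara skydiving contrail cosmetic that attaches skull-and-spine trails to the character's hands and mesh and ribbon trails to the feet.
# Hand trails are fixed-count ribbon chains, re-anchored to bones each update and displaced by velocity and sine waves.
# Spine meshes read those ribbon particles to follow them as oriented chain links, cycling through shrinking sizes.
# Contrail setup's dive value, glide lean and front-end modifiers adjust shape and scale.
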NiagaraSystem /BRCosmetics/Athena/Effects/Particles/Skydiving/NS_Contrail_TallChore/NS_Contrail_TallChore
dependencies {
 usedDataInterfaces [
  &NiagaraDataInterfaceArrayFloat3
  &NiagaraDataInterfaceArrayFloat
  &NiagaraDataInterfaceCurve
  &NiagaraDataInterfaceCamera
  &NiagaraDataInterfaceParticleRead
  &NiagaraDataInterfaceArrayInt32
  &NiagaraDataInterfaceSkeletalMesh
 ]
 usedDynamicInputs []
 usedModules [
  &self:SystemSpawnScript.SystemScriptSource.SystemScriptGraph.NiagaraNodeAssignment_7.SetVariables_#1
  &self:SystemSpawnScript.SystemScriptSource.SystemScriptGraph.NiagaraNodeAssignment_0.SetVariables_#2
  &SystemState
  &self:SystemSpawnScript.SystemScriptSource.SystemScriptGraph.NiagaraNodeAssignment_1.SetVariables_#3
  /Game/Effects/Niagara/Modules/Contrails/FN_Contrail_Setup
  &FN_Integrator
 ]
}
emitters {
 NE_Feet_Additional {
  emitterUpdateScript [
   &EmitterState()
   &SpawnRate(SpawnRate=20, Use Spawn Probability=false)
  ]
  particleSpawnScript [
   &InitializeParticle(Color=<rgba 0.1183 0.3 0.1298 1>, Ribbon Width=30, Ribbon Width Mode=::NewEnumerator1, UsePositionOffset=false, Write Material Random=true)
   &AddVelocity(
    Velocity=&MakeVector(X=0, Y=0, Z=1)
    Velocity Speed Scale=&Lerp_Float(A=100, B=200, Alpha=$User.ParaGlideLeanAlpha)
   )
   &PartitionParticles()
  ]
  particleUpdateScript [
   &ParticleState()
   &SolveForcesAndVelocity(Clamp Velocity=false, Limit Acceleration=false)
   &SkeletalMeshLocation(
    Bone / Socket Index=&SelectIntFromArray(Direct Array Index=$Particles.PartitionParticles.Partition, Int Selection Array=DataInterface() {propertyValues {IntData [2,3]}})
    MaskSpawnGroup=false
    Mesh Sampling Type=::NewEnumerator5
    Position Sampling=::NewEnumerator1
    Skeletal Mesh=$System.Skeletal Mesh
    Socket Sampling Mode=::NewEnumerator5
    Source Movement Scale:Vector3f=<0.95 0.95 0.95>
   )
   &ScaleRibbonWidth(Ribbon Width Scale=&FloatFromCurve(FloatCurve=DataInterface() {propertyValues {Curve {keys [<at 0 0>,<at 0.2 1>,<at 1 0.5>]}}}, Scale Curve=1.25))
   &ScaleColor(
    Scale Alpha=&FloatFromCurve(FloatCurve=DataInterface() {propertyValues {Curve {keys [<at 0 0>,<at 0.2 1>,<at 1 0>]}}})
    ScaleA=true
    ScaleRGB=true
   )
  ]
  rendererClasses [&NiagaraRibbonRendererProperties]
  renderers [{rendererClass=&NiagaraRibbonRendererProperties}]
 }
 NE_HandSkull {
  emitterUpdateScript [
   &EmitterState(
    Inactive Response=::NewEnumerator1
    Life Cycle Mode=::NewEnumerator1
    Loop Behavior=::NewEnumerator1
    Loop Duration Mode=::NewEnumerator1
    UseLoopDelay=false
   )
   &SpawnBurst_Instantaneous(Spawn Count=2, Use Spawn Probability=false, UseLoopCountLimit=false)
  ]
  particleSpawnScript [
   &InitializeParticle(
    Color Mode=::NewEnumerator0
    Lifetime=5
    Mesh Scale Mode=::NewEnumerator3
    Sprite Rotation Mode=::NewEnumerator1
    Sprite Size Mode=::NewEnumerator3
    Uniform Sprite Size=50
    UsePositionOffset=false
    Write Material Random=true
   )
   &PartitionParticles()
   &SkeletalMeshLocation(
    Bone / Socket Index=$Particles.PartitionParticles.Partition
    Bone Sampling Mode=::NewEnumerator4
    MaskSpawnGroup=false
    Sampled Position Offset Coordinate Space:ENiagaraSkelMeshTransforms=::NewEnumerator1
    Skeletal Mesh=$System.Skeletal Mesh
   )
   /SetVariables() {isSetParameters=true}
   /SetVariables(Particles.EnergyHandsColor:LinearColor=<rgba 1 1 1 0.4>) {isSetParameters=true}
  ]
  particleUpdateScript [
   &ParticleState(Kill Particles When Lifetime Has Elapsed=false)
   &SkeletalMeshLocation(
    Bone / Socket Index=$Particles.PartitionParticles.Partition
    Bone Sampling Mode=::NewEnumerator4
    MaskSpawnGroup=false
    Orientation Sampling=::NewEnumerator1
    Sampled Position Offset=&Lerp_Vector(
     A=&Lerp_Vector(A=<10 0 -5>, B=<-10 0 5>, Alpha=&MakeFloatFromInt(Int=$Particles.PartitionParticles.Partition))
     B=&Lerp_Vector(A=<10 0 0>, B=<-10 0 0>, Alpha=&MakeFloatFromInt(Int=$Particles.PartitionParticles.Partition))
     Alpha=$User.ParaGlideLeanAlpha
    )
    Skeletal Mesh=$System.Skeletal Mesh
   ) {module=SkeletalMeshLocation001}
   /SetVariables(Particles.MeshOrientation=&MultiplyQuaternion(
     Quaternion A=$Particles.SkeletalMeshLocation001.SampledRotation
     Quaternion B=&LerpQuaternion(
      Quaternion A:Quat4f=&MultiplyQuaternion(
       Quaternion A=&MakeQuatFromAxisAngle(Axis=<0 1 0>, AngleInDegrees=-90)
       Quaternion B=&MakeQuatFromAxisAngle(Axis=<1 0 0>, AngleInDegrees=-105)
      )
      Quaternion B:Quat4f=&MultiplyQuaternion(
       Quaternion A=&MakeQuatFromAxisAngle(Axis=<0 1 0>, AngleInDegrees=-90)
       Quaternion B=&MakeQuatFromAxisAngle(Axis=<1 0 0>, AngleInDegrees=75)
      )
      Lerp Factor:NiagaraFloat=&MakeFloatFromInt(Int=$Particles.PartitionParticles.Partition)
      Enable Spherical Linear Interpolation:NiagaraBool=false
     )
    )) {enabled=false, isSetParameters=true}
   /SetVariables(Particles.MeshOrientation=<-1 0 0 1>) {isSetParameters=true}
   /SetVariables(Particles.SpriteAlignment=&NormalizeVector(
     Vector To Normalize=&Add_Vector(
      A=&Add_Vector(
       A=$Particles.SpriteAlignment
       B=&NormalizeVector(
        Vector To Normalize=&Multiply_Vector(A=$Engine.Owner.Velocity, B=<1 1 -1>)
        Safe Normalize=true
        Safe Fallback Threshold=0
        Safe Fallback Vector=<0 0 0.001>
       )
      )
      B=<0 0 0.4>
     )
     Safe Normalize=true
     Safe Fallback Threshold=0
     Safe Fallback Vector=<0 0 0.001>
    )) {isSetParameters=true}
   &SolveForcesAndVelocity(Clamp Velocity=false, Limit Acceleration=false)
   &ScaleColor(ScaleA=true, ScaleRGB=true)
  ]
  rendererClasses [&NiagaraMeshRendererProperties]
  renderers [
   {rendererClass=&NiagaraMeshRendererProperties, rendererIndex=0}
   {rendererClass=&NiagaraMeshRendererProperties, rendererIndex=1}
  ]
 }
 NE_Spines {
  emitterUpdateScript [
   &EmitterState(
    Inactive Response=::NewEnumerator1
    Life Cycle Mode=::NewEnumerator1
    Loop Behavior=::NewEnumerator1
    Loop Duration Mode=::NewEnumerator1
    UseLoopDelay=false
   )
   &SpawnBurst_Instantaneous(
    Spawn Count=&Multiply_Int(A=$System.ChainNumber, B=$System.LinkPerChain)
    Use Spawn Probability=false
    UseLoopCountLimit=false
   )
  ]
  particleSpawnScript [
   &FN_SetSpawnGroup(Number of Spawn Groups=2)
   &InitializeParticle(Color Mode=::NewEnumerator0, Lifetime=5, Mesh Scale Mode=::NewEnumerator3, UsePositionOffset=false, Write Material Random=true)
   &PartitionParticles(
    Number Of Partitions=$System.ChainNumber
    Particles Per Partition:NiagaraInt32=$System.LinkPerChain
    Partition Mode:ENiagara_PartitionMode=::NewEnumerator1
   )
   /SetVariables(Particles.NormalizedChainPosition:NiagaraFloat=$Particles.PartitionParticles.Normalized Index In Partition) {isSetParameters=true, module=SetVariables_#4}
   &SkeletalMeshLocation(
    Bone / Socket Index=$Particles.PartitionParticles.Partition
    Bone Sampling Mode=::NewEnumerator4
    MaskSpawnGroup=false
    Sampled Position Offset=&Lerp_Vector(A=<0 0 0>, B=<0 0 100>, Alpha=$Particles.NormalizedChainPosition)
    Sampled Position Offset Coordinate Space:ENiagaraSkelMeshTransforms=::NewEnumerator1
    Skeletal Mesh=$System.Skeletal Mesh
   )
   /SetVariables(Particles.MeshOrientation=&MultiplyQuaternion(
     Quaternion A=&MultiplyQuaternion(
      Quaternion A=&OrientMeshToVector(
       Look At Direction=&NormalizeVector(
        Vector To Normalize=&Multiply_VectorByFloat(
         Vector=<0 0 1>
         Float=&Lerp_Float(A=-1, B=1, Alpha=&MakeFloatFromInt(Int=$Particles.PartitionParticles.Partition))
        )
        Safe Normalize=true
        Safe Fallback Threshold=0
        Safe Fallback Vector=<0 0 0.001>
       )
       Reference Axis Vector=<1 0 0>
      )
      Quaternion B=&MultiplyQuaternion(
       Quaternion A=&MakeQuatFromAxisAngle(
        Axis=<0 1 0>
        AngleInDegrees=&Lerp_Float(A=90, B=-90, Alpha=&MakeFloatFromInt(Int=$Particles.PartitionParticles.Partition))
       )
       Quaternion B=&MakeQuatFromAxisAngle(
        Axis=<0 0 1>
        AngleInDegrees=&Lerp_Float(
         A=180
         B=&Lerp_Float(A=270, B=90, Alpha=&MakeFloatFromInt(Int=$Particles.PartitionParticles.Partition))
         Alpha=&FloatFromCurve(
          FloatCurve=DataInterface() {propertyValues {Curve {keys [
              {arriveTangent=2.7188, interpMode=RCIM_Cubic, leaveTangent=2.7188, tangentMode=RCTM_User, time=-1, value=-1}
              {interpMode=RCIM_Cubic, tangentMode=RCTM_User, time=0, value=0}
              {arriveTangent=2.2136, interpMode=RCIM_Cubic, leaveTangent=2.2136, tangentMode=RCTM_User, time=1, value=0.8}
             ]}}}
          CurveIndex=&DotProduct(
           Vector A=&NormalizeVector(
            Vector To Normalize=&Multiply_Vector(
             A=&Subtract_Position(
              A=$Particles.Position
              B=&ReturnCameraPosition(Coordinate Space:ENiagaraCoordinateSpace=::Simulation, Camera Query:NiagaraDataInterfaceCamera=DataInterface())
             )
             B=<1 1 1>
            )
            Safe Normalize=true
            Safe Fallback Threshold=0
            Safe Fallback Vector=<0 0 0.001>
           )
           Vector B=<0 0 1>
          )
         )
        )
        NormalizeAxis=true
       )
      )
     )
     Quaternion B=<0 0 0 1>
    )) {isSetParameters=true}
   &CameraOffset(Camera Offset Amount=10)
  ]
  particleUpdateScript [
   /SetVariables(Particles.SmokeColor:LinearColor=<rgba 0.2745 1 0.4856 0.5826>) {isSetParameters=true}
   &ParticleState(Kill Particles When Lifetime Has Elapsed=false)
   &SkeletalMeshLocation(
    Bone / Socket Index=$Particles.PartitionParticles.Partition
    Bone Sampling Mode=::NewEnumerator4
    MaskSpawnGroup=false
    Position Sampling=::NewEnumerator2
    Sampled Position Offset=&Lerp_Vector(A=<20 -3 0>, B=<-20 3 1>, Alpha=&MakeFloatFromInt(Int=$Particles.PartitionParticles.Partition))
    Skeletal Mesh=$System.Skeletal Mesh
   ) {module=SkeletalMeshLocation001}
   &self:StickMeshToRibbon_04(
    Index:NiagaraInt32=$Particles.PartitionParticles.Index In Partition
    LinkOffset=3
    Particle Attribute Reader=DataInterface() {propertyValues {EmitterBinding {bindingMode=Other, emitterName=NE_Trail_Hands}}}
    ParticlePerRibbon:NiagaraInt32=$System.ParticlePerRibbon
    Partition:NiagaraInt32=$Particles.PartitionParticles.Partition
    RibbonSpace=4
   ) {module=StickMeshToRibbon_04}
   &InterpolateOverTime(
    Rate Of Change:NiagaraFloat=&FN_FrontEndFloatModifier(Gameplay=1, Front End=10)
    Target Vector Value:Vector3f=&Multiply_VectorByFloat(
     Vector=&NormalizeVector(
      Vector To Normalize=&Multiply_Vector(
       A=&ReturnCameraVector(Camera Vectors=::NewEnumerator4, Coordinate Space=::Simulation, Origin Position=$Particles.Position)
       B=<1 1 1>
      )
      Safe Normalize=true
      Safe Fallback Threshold=0
      Safe Fallback Vector=<0 0 0.001>
     )
     Float=-1
    )
    Variable Type=::NewEnumerator2
   )
   &InterpolateOverTime(
    Rate Of Change=10
    Target Vector Value:Vector3f=&Multiply_VectorByFloat(Vector=$Particles.RibbonDirectionVector, Float=-1)
    Variable Type=::NewEnumerator2
   ) {module=InterpolateOverTime001}
   &self:OrientChainLinks_04(
    FrontVector:Vector3f=$Particles.InterpolateOverTime.InterpolateOverTime_V3_.Moving Average
    RightVector:Vector3f=<1 0 0>
    UpVector:Vector3f=$Particles.InterpolateOverTime001.InterpolateOverTime_V3_.Moving Average
   ) {module=OrientChainLinks_04}
   &SolveForcesAndVelocity(Clamp Velocity=false, Limit Acceleration=false)
   /SetVariables(Particles.MeshIndex=&ModuloInt(A=$Particles.PartitionParticles.Index In Partition, B=7)) {isSetParameters=true}
   /SetVariables(Particles.SpineSize:Vector3f=&MergeHelper_VectorSwitch(
     0=<1 1 1>
     1=<0.9 0.9 0.9>
     2=<0.8 0.8 0.8>
     3=<0.7 0.7 0.7>
     4=<0.6 0.6 0.6>
     5=<0.5 0.5 0.5>
     6=<0.4 0.4 0.4>
     Index=$Particles.MeshIndex
     Num=::NewEnumerator5
    )) {isSetParameters=true}
  ]
  rendererClasses [&NiagaraMeshRendererProperties]
  renderers [
   {rendererClass=&NiagaraMeshRendererProperties, rendererIndex=0}
   {rendererClass=&NiagaraMeshRendererProperties, rendererIndex=1}
  ]
 }
 NE_Trail_Feet {
  emitterUpdateScript [
   &EmitterState(Life Cycle Mode=::NewEnumerator1, Loop Behavior=::NewEnumerator1, Loop Duration Mode=::NewEnumerator1, UseLoopDelay=false)
   &SpawnBurst_Instantaneous(Spawn Count=2, Use Spawn Probability=true, UseLoopCountLimit=false)
  ]
  particleSpawnScript [
   &PartitionParticles()
   &InitializeParticle(
    Color Mode=::NewEnumerator0
    Mesh Scale=<1.2 1 1>
    Mesh Scale Mode=::NewEnumerator4
    Sprite Size Mode=::NewEnumerator3
    Uniform Sprite Size=65
    UsePositionOffset=false
    Write Material Random=true
   )
   &InitialMeshOrientation(Mesh Orientation Mode=::NewEnumerator2, Rotation=<0 -0.25 0>, Use Rotation Vector=true)
   &InitialMeshOrientation(
    Mesh Orientation Mode=::NewEnumerator2
    Rotation=&SelectVectorFromArray(
     Direct Array Index=$Particles.PartitionParticles.Partition
     Vector Selection Array=DataInterface() {propertyValues {FloatData [<0 0 0>, <0 0 0>, <0 0 0>, <0 0 0>]}}
    )
    Use Rotation Vector=true
   ) {module=InitialMeshOrientation001}
  ]
  particleUpdateScript [
   &ParticleState(Kill Particles When Lifetime Has Elapsed=false, Let Infinitely Lived Particles Die When Emitter Deactivates=true)
   &SkeletalMeshLocation(
    Bone / Socket Index=&SelectIntFromArray(Direct Array Index=$Particles.PartitionParticles.Partition, Int Selection Array=DataInterface() {propertyValues {IntData [2,3]}})
    MaskSpawnGroup=false
    Mesh Sampling Type=::NewEnumerator5
    Orientation Sampling=::NewEnumerator1
    Sampled Position Offset=&MergeHelper_VectorSwitch(0=<0 0 0>, 1=<0 0 0>, 2=?, 3=?, Index=$Particles.PartitionParticles.Partition, Num=::NewEnumerator0)
    Skeletal Mesh=$System.Skeletal Mesh
    Socket Sampling Mode=::NewEnumerator5
   )
   &ScaleMeshSize(Scale Factor=<0.8 1 1>) {module=ScaleMeshSize001}
   &ScaleMeshSize(Scale Factor=&MakeVector(
     X=&Lerp_Float(
      A=&FN_FrontEndFloatModifier(Gameplay=0.5, Front End=0.7)
      B=&FN_FrontEndFloatModifier(Gameplay=1.25, Front End=0.8)
      Alpha=&ClampFloat(Float=$System.FN_Contrail_Setup.Dive Value)
     )
     Y=1
     Z=1
    ))
   /SetVariables(
    Particles.Pink_DynamParam1:Vector4f=&MakeVector4(
     X=&FloatFromCurve(
      FloatCurve=DataInterface() {propertyValues {Curve {keys [<at 0 1>,<at 1 0>]}}}
      CurveIndex=&Divide_Float(A=&VectorLength(Vector=$Engine.Owner.Velocity), B=9000)
     )
     Y=&MakeCustomFloatFromBool(Bool=&SetBoolByFloatComparison(A=$System.FN_Contrail_Setup.Dive Value, B=0.2), True Float=1.4, False Float=1)
     Z=&SelectFloatFromArray(
      Direct Array Index=$Particles.PartitionParticles.Partition
      Float Selection Array=DataInterface() {propertyValues {FloatData [0,0.2,0.4,0.6]}}
     )
     W=0
    )
    Particles.Pink_DynamParam2:Vector4f=&MakeVector4(
     X=&FN_FrontEndFloatModifier(
      Gameplay=&Lerp_Float(
       A=&SelectFloatFromArray(
        Direct Array Index=$Particles.PartitionParticles.Partition
        Float Selection Array=DataInterface() {propertyValues {FloatData [0.8,0.8,0.35,0.35]}}
       )
       B=&SelectFloatFromArray(
        Direct Array Index=$Particles.PartitionParticles.Partition
        Float Selection Array=DataInterface() {propertyValues {FloatData [0.7,0.7,0.3,0.3]}}
       )
       Alpha=&ClampFloat(Float=$System.FN_Contrail_Setup.Dive Value)
      )
      Front End=0.7
     )
     Y=0
     Z=0
     W=0
    )
   ) {isSetParameters=true}
  ]
  rendererClasses [&NiagaraMeshRendererProperties]
  renderers [{rendererClass=&NiagaraMeshRendererProperties, rendererIndex=0}]
 }
 NE_Trail_Hands {
  emitterUpdateScript [
   &EmitterState(
    Inactive Response=::NewEnumerator1
    Life Cycle Mode=::NewEnumerator1
    Loop Behavior=::NewEnumerator1
    Loop Duration Mode=::NewEnumerator1
    UseLoopDelay=false
   )
   &SpawnBurst_Instantaneous(
    Spawn Count=&Multiply_Int(A=$System.ChainNumber, B=$System.ParticlePerRibbon)
    Use Spawn Probability=false
    UseLoopCountLimit=false
   )
  ]
  particleSpawnScript [
   &InitializeParticle(
    Color Mode=::NewEnumerator0
    Lifetime=5
    Mesh Scale Mode=::NewEnumerator4
    Ribbon Facing Mode=::NewEnumerator1
    Ribbon Facing Vector:Vector3f=<1 0 0>
    Sprite Size Mode=::NewEnumerator3
    UsePositionOffset=false
    Write Material Random=true
   )
   &PartitionParticles(
    Number Of Partitions=$System.ChainNumber
    Particles Per Partition:NiagaraInt32=$System.ParticlePerRibbon
    Partition Mode:ENiagara_PartitionMode=::NewEnumerator1
   )
   /SetVariables(Particles.NormalizedChainPosition:NiagaraFloat=$Particles.PartitionParticles.Normalized Index In Partition) {isSetParameters=true, module=SetVariables_#4}
   /SetVariables() {isSetParameters=true}
   /SetVariables(
    Particles.FirstRibbonColor:LinearColor=<rgba 1 1 1 0.5>
    Particles.SecondRibbonColor:LinearColor=<rgba 0.0102 0.2 0 2.5>
    Particles.ThirdRibbonColor:LinearColor=<rgba 0 0.5 0.0435 5>
   ) {isSetParameters=true}
   /SetVariables(
    Particles.SecondRibbonWidth:NiagaraFloat=&FN_FrontEndFloatModifier(Gameplay=40, Front End=30)
    Particles.ThirdRibbonWidth:NiagaraFloat=&FN_FrontEndFloatModifier(Gameplay=40, Front End=25)
   ) {isSetParameters=true}
  ]
  particleUpdateScript [
   &ParticleState(Kill Particles When Lifetime Has Elapsed=false)
   &SkeletalMeshLocation(
    Bone / Socket Index=$Particles.PartitionParticles.Partition
    Bone Sampling Mode=::NewEnumerator4
    MaskSpawnGroup=false
    Orientation Sampling=::NewEnumerator1
    Sampled Position Offset=&FN_FrontEndVector3Modifier(
     Gameplay=&MergeHelper_VectorSwitch(0=<0 0 0>, 1=<0 0 0>, 2=?, 3=?, Index=$Particles.PartitionParticles.Partition, Num=::NewEnumerator0)
     Front End=&Lerp_Vector(A=<-1 -2 0>, B=<0 4 0>, Alpha=$System.FN_Contrail_Setup.Dive Value)
    )
    Sampled Position Offset Coordinate Space:ENiagaraSkelMeshTransforms=::NewEnumerator2
    Skeletal Mesh=$System.Skeletal Mesh
   ) {module=SkeletalMeshLocation002}
   /SetVariables(Particles.Position=&AddVectorToPosition(
     Position=$Particles.Position
     Vector=&Multiply_VectorByFloat(
      Vector=&Add_Vector(
       A=&Multiply_Vector(
        A=&NormalizeVector(
         Vector To Normalize=&FN_FrontEndVector3Modifier(Gameplay=$System.PlayerVelDir, Front End=$System.FrontendVelocity)
         Safe Normalize=true
         Safe Fallback Threshold=0
         Safe Fallback Vector=<0 0 0.001>
        )
        B=&SelectVectorFromArray(
         Direct Array Index=$Particles.PartitionParticles.Partition
         Vector Selection Array=DataInterface() {propertyValues {FloatData [<-0.5 -0.5 -1>, <-0.5 -0.5 -1>, <0.1 -1.5 -1>, <0.1 -0.5 -1>, <-0.5 -1.5 -1>, <-0.5 -1 -1>, <0.1 -1 -1>, <0.1 -0.5 -1>]}}
        )
       )
       B=&Multiply_VectorByFloat(
        Vector=&TransformVector(
         Vector=&SelectVectorFromArray(
          Direct Array Index=$Particles.PartitionParticles.Partition
          Direct Array Mode=::NewEnumerator1
          Vector Selection Array=DataInterface() {propertyValues {FloatData [<2 -2 0>, <-2 0 0>, <1 2 0>, <-0.5 0 0>, <0.5 0 0>, <-1 2 0>, <1 1 0>, <-2 1 0>]}}
         )
         Destination Space=::Simulation
        )
        Float=&SelectFloatFromArray(
         Direct Array Index=&SpawnGroup()
         Float Selection Array=DataInterface() {propertyValues {FloatData [0.1,0.15,0.08,0.16,0.15,0.18]}}
        )
       )
      )
      Float=&FloatFromCurve(
       FloatCurve=DataInterface() {propertyValues {Curve {keys [{tangentMode=RCTM_User, time=0, value=0}, {tangentMode=RCTM_User, time=1, value=1}]}}}
       CurveIndex=$Particles.PartitionParticles.Normalized Index In Partition
       Scale Curve=&SelectFloatFromArray(
        Direct Array Index=&SpawnGroup()
        Float Selection Array=DataInterface() {propertyValues {FloatData [97,97,120,100,75,55,94,80]}}
       )
      )
     )
    )) {isSetParameters=true}
   /SetVariables(Particles.Position=&AddVectorToPosition(
     Position=$Particles.Position
     Vector=&Multiply_VectorByFloat(
      Vector=/Niagara/DynamicInputs/Vectors/RotateVector(
       Vector To Rotate:Vector3f=<10 0 3>
       Yaw:NiagaraFloat=&Add_Float(
        A=$Particles.Age
        B=&Multiply_Float(A=&MakeFloatFromInt(Int=&SpawnGroup()), B=2)
       )
       Pitch:NiagaraFloat=0
       Roll:NiagaraFloat=0
       Rotation Rate:NiagaraFloat=&SelectFloatFromArray(
        Direct Array Index=$Particles.PartitionParticles.Partition
        Float Selection Array=DataInterface() {propertyValues {FloatData [100,100,130,120,160,150,120,140]}}
       )
      )
      Float=&Multiply_Float(
       A=&Sine(
        Normalized Angle=&Add_Float(
         A=&Multiply_Float(
          A=$Emitter.Age
          B=&SelectFloatFromArray(
           Direct Array Index=&SpawnGroup()
           Float Selection Array=DataInterface() {propertyValues {FloatData [-2,-2.5,-2.5,-3,-3,-2,-2.5,-2]}}
          )
         )
         B=$Particles.PartitionParticles.Normalized Index In Partition
        )
        Period=0.9
        Scale=&Lerp_Float(A=0.8, B=2, Alpha=$System.FN_Contrail_Setup.Dive Value)
        Bias=0
       )
       B=&FloatFromCurve(
        FloatCurve=DataInterface() {propertyValues {Curve {keys [
            {arriveTangent=1.1676, interpMode=RCIM_Cubic, leaveTangent=1.1676, tangentMode=RCTM_User, time=0, value=0}
            {arriveTangent=1.4544, interpMode=RCIM_Cubic, leaveTangent=1.4544, tangentMode=RCTM_User, time=1, value=1}
           ]}}}
        CurveIndex=$Particles.PartitionParticles.Normalized Index In Partition
       )
      )
     )
    )) {isSetParameters=true}
   /SetVariables(Particles.Position=&AddVectorToPosition(
     Position=$Particles.Position
     Vector=&Multiply_VectorByFloat(
      Vector=&SelectVectorFromArray(
       Direct Array Index=$Particles.PartitionParticles.Partition
       Direct Array Mode=::NewEnumerator1
       Vector Selection Array=DataInterface() {propertyValues {FloatData [<-20 40 130>, <40 20 130>, <-25 0 0>, <25 0 0>]}}
      )
      Float=&FloatFromCurve(
       FloatCurve=DataInterface() {propertyValues {Curve {keys [{arriveTangent=0.9713, time=0, value=0}, <at 0.125 0.2>, <at 0.95 0.92>, {interpMode=RCIM_Cubic, leaveTangent=0.018, time=1, value=1}]}}}
       CurveIndex=$Particles.Initial.PartitionParticles.Normalized Index In Partition
       Scale Curve=&Lerp_Float(A=1, B=1.5, Alpha=$User.ParaGlideLeanAlpha)
      )
     )
    )) {isSetParameters=true}
   &DynamicMaterialParameters(
    Index 0 Param 1=1
    Index 0 Param 2=1
    Index 0 Param 3=2
    Index 0 Param 4=1
    Module.Index0Float0Write=true
    Module.Index0Float1Write=true
    Module.Index0Float2Write=true
    Module.Index0Float3Write=true
   )
  ]
  rendererClasses [&NiagaraRibbonRendererProperties]
  renderers [
   {rendererClass=&NiagaraRibbonRendererProperties}
   {rendererClass=&NiagaraRibbonRendererProperties, rendererIndex=1}
   {rendererClass=&NiagaraRibbonRendererProperties, rendererIndex=2}
  ]
 }
}
summary {
 systemName=NS_Contrail_TallChore
 emitters [
  {emitterName=NE_HandSkull, rendererClasses [&NiagaraMeshRendererProperties]}
  {emitterName=NE_Trail_Hands, rendererClasses [&NiagaraRibbonRendererProperties]}
  {emitterName=NE_Spines, rendererClasses [&NiagaraMeshRendererProperties]}
  {emitterName=NE_Trail_Feet, rendererClasses [&NiagaraMeshRendererProperties]}
  {emitterName=NE_Feet_Additional, rendererClasses [&NiagaraRibbonRendererProperties]}
 ]
 userVariables [
  User.bIsFrontEnd {type=NiagaraBool}
  User.bIsFrontEndPreview {type=NiagaraBool}
  User.Debug {type=NiagaraBool}
  User.DebugVelocity {type=Vector3f, description="", defaultValue {struct=Vector3f, value=<0 0 1>}}
  User.EmissiveColor {type=LinearColor, description="", defaultValue {struct=LinearColor, value=<rgba 0.9805 0 1 1>}}
  User.IsReacting {type=NiagaraFloat, defaultValue=0, description=""}
  User.IsSkydiving {type=NiagaraBool}
  User.ParaGlideLeanAlpha {type=NiagaraFloat, defaultValue=0, description=""}
  User.RibbonIDOffset {type=NiagaraInt32}
 ]
}
